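# Parent skin material for the WaistRush backpack, adding an emissive digit display to the 2023 character parent material.
# Pre-FX output gets a metal roughness tweak and car-style clear coat before the display and post-FX stages.
# Digits reuse AgileRug's number function and mask, with bump-offset depth and tinting.
# A custom-primitive-data timestamp builds a timed rise-and-fall animation signal alongside an emissive boost blend.
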Material /BRCosmetics/Accessories/FORT_Backpacks/Backpack_M_MED_WaistRush/Skins/M_MED_WaistRush_Backpack_Skins_Parent
graph {
 properties {
  bUseMaterialAttributes=true
  bUsedWithMeshDeformer=true
  bUsedWithMorphTargets=true
  bUsedWithSkeletalMesh=true
  pixelDepthOffsetMode=PDOM_Legacy
  shadingModel=MSM_FromMaterialExpression
  usedShadingModels=Default Lit | Subsurface | Clear Coat | Subsurface Profile | Cloth
 }
 nodes [
  n1 = MaterialFunctionCall() {materialFunction=&MF_CharacterParent_2023_PreFX}
  n2 = GetMaterialAttributes(None=@n1.Result) {attributeGetTypes [#1, #2]}
  n3 = Multiply(A=@n2.Roughness, B=ScalarParameter() {defaultValue=1, parameterName=Metal Roughness Multiplier})
  n4 = Saturate(None=Power(Base=@n2.Metallic))
  n5 = MaterialFunctionCall!(
   Mask=@n1.Tech Art Mask
   Fresnel=@n1.Tech Art Mask
   In=SetMaterialAttributes!(MaterialAttributes=@n2.MaterialAttributes, Roughness=LinearInterpolate(A=@n2.Roughness, B=@n3, Alpha=@n4)) {attributeSetTypes [#1]}
  ) {materialFunction=/Game/Global/MaterialFunctions/TechArt/MF_Advanced_CarClearCoat}
  ClearCoatNormalCustomOutput!(None=MaterialFunctionCall(BaseNormal=@n5.NormalWithOrangePeel, AdditionalNormal=@n5.NormalWithOrangePeel) {materialFunction=&BlendAngleCorrectedNormals}.Result)
  NamedRerouteDeclaration(None=@n1.Tech Art Mask) {name=TA_Mask, nodeColor=<rgba 0.5 0 1 1>}
  NamedRerouteDeclaration(None=@n5.Result) {name=IN_Material, nodeColor=<rgba 1 0 0.2109 1>}
  n6 = GetMaterialAttributes(None=NamedRerouteUsage()) {attributeGetTypes [#3]}
  n7 = VectorParameter() {group=WaistRush, parameterName=DigitsColor, defaultValue=<rgba 0 0.2384 0.3968 1>}
  n8 = MaterialFunctionCall(
   UVs=BumpOffset(Coordinate=TextureCoordinate(), Height=ScalarParameter() {defaultValue=-0.25, group=AgileRug, parameterName=GraphicsDepth}) {heightRatio=0.01}
   "Rotation Angle (0-1)"=Constant() {r=0.25}
  ) {materialFunction=&CustomRotator}
  n9 = Constant2Vector() {g=-0.8775, r=0.038}
  n10 = Saturate(None=Multiply(A=Add(A=@n8.Rotated Values, B=@n9), B=Constant2Vector() {g=34, r=3.5}))
  n11 = BumpOffset(Coordinate=TextureCoordinate(), Height=ScalarParameter() {defaultValue=-0.25, group=AgileRug, parameterName=GraphicsDepth}) {heightRatio=0.01}
  n12 = TextureObjectParameter() {
   group=AgileRug
   parameterName=DisplayMask
   samplerType=SAMPLERTYPE_LinearColor
   texture=/BRCosmetics/Accessories/FORT_Backpacks/Backpack_M_MED_AgileRug/Textures/T_M_MED_AgileRug_Backpack_Mask
  }
  n13 = TextureCoordinate()
  n14 = Multiply(
   A=Max(
    A=MaterialFunctionCall!(UVs=@n10) {materialFunction=/BRCosmetics/Accessories/FORT_Backpacks/Backpack_M_MED_AgileRug/Materials/MF_NumberDisplay}.Result
    B=TextureSample(UVs=@n11, Tex=@n12) {samplerSource=SSM_Clamp_WorldGroupSettings}.R
   )
   B=NamedRerouteDeclaration(None=TextureSample(UVs=@n13, Tex=@n12) {samplerSource=SSM_Clamp_WorldGroupSettings}.B) {name=EffectMask, nodeColor=<rgba 0 0.4609 1 1>}
  )
  n15 = Max(A=@n6.Emissive Color, B=Multiply(A=Constant() {r=3}, B=Multiply(A=@n7.RGB, B=@n14)))
  n16 = NamedRerouteUsage()
  n17 = Max(A=@n15, B=Multiply(A=Multiply(A=@n15, B=@n16)) {constB=3})
  n18 = NamedRerouteUsage()
  NamedRerouteDeclaration(None=SetMaterialAttributes!(MaterialAttributes=@n6.MaterialAttributes, Emissive Color=LinearInterpolate(A=@n15, B=@n17, Alpha=@n18)) {attributeSetTypes [#3]}) {name=OUT_Material, nodeColor=<rgba 0 0.5312 1 1>}
  n19 = ScalarParameter() {bUseCustomPrimitiveData=true, parameterName=TimeStamp_CPD, primitiveDataIndex=4}
  n20 = Abs(None=Subtract(A=Time(), B=@n19))
  n21 = ScalarParameter() {defaultValue=1, group=AgileRug, parameterName=EffectSpeed}
  n22 = Multiply(A=Saturate(None=Divide(A=@n20, B=@n21)), B=Saturate(None=@n19))
  n23 = Time()
  n24 = Add(
   A=ScalarParameter() {bUseCustomPrimitiveData=true, parameterName=TimeStamp_CPD, primitiveDataIndex=4}
   B=ScalarParameter() {defaultValue=7, parameterName=EffectDelay}
  ) {constB=5}
  n25 = Saturate(None=Divide(A=Subtract(A=@n23, B=@n24), B=ScalarParameter() {defaultValue=1, group=AgileRug, parameterName=EffectSpeed}))
  NamedRerouteDeclaration(None=Saturate(None=Subtract(A=@n22, B=@n25))) {name=TimeStampAnimation, nodeColor=<rgba 0.7813 0 1 1>}
  SetMaterialAttributes!(
   MaterialAttributes=MaterialFunctionCall(Material=NamedRerouteUsage(), M=@n1.M, Pre Skinned Local Position=@n1.M, Distance Blend=@n1.M, Fresnel=@n1.M) {materialFunction=&MF_CharacterParent_2023_PostFX}.Result
   Normal=GetMaterialAttributes(None=@n1.Result) {attributeGetTypes [#4]}.Normal
  ) {attributeSetTypes [#4]}
 ]
}
